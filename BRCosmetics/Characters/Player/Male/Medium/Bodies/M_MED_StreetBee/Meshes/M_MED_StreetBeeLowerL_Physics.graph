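# Physics asset for the StreetBee medium male player body's LowerL mesh, built around a dangling strap on the left leg.
# Capsules on pelvis, thighs, thigh twists, right calf and both feet form a fully angular-locked body chain, with foot_l constrained straight to thigh_l as no left calf body exists.
# A locked base strap bone on thigh_l anchors two chained dyn_legStrap capsules with small swing angles and zero twist angle.
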PhysicsAsset /BRCosmetics/Characters/Player/Male/Medium/Bodies/M_MED_StreetBee/Meshes/M_MED_StreetBeeLowerL_Physics
bodies {
 base_legStrap_00_l [{shapeType=Capsule, length=0.505, radius=0.505, shapeName=base_legStrap_00_l_capsule, rotation=<rot 0 -90 0>}]
 calf_r [{shapeType=Capsule, length=20.0435, radius=8.5, center=<0 0.8482 0>, rotation=<rot 90 180 0>}]
 dyn_legStrap_00_l [{shapeType=Capsule, length=6, radius=2, shapeName=dyn_legStrap_00_l_capsule, center=<4.5 1 0.3>, rotation=<rot 90 -4.9999 0>}]
 dyn_legStrap_01_l [{shapeType=Capsule, length=5.8677, radius=2, shapeName=dyn_legStrap_01_l_capsule, center=<3 0.8183 0.3>, rotation=<rot 90 0 0>}]
 foot_l [{shapeType=Capsule, length=15.6916, radius=10, rotation=<rot 90 0 0>}]
 foot_r [{shapeType=Capsule, length=15.6491, radius=10, rotation=<rot 90 0 0>}]
 pelvis [{shapeType=Capsule, length=20.3917, radius=9.5, center=<-17.5121 -0.6527 0>}]
 thigh_l [{shapeType=Capsule, length=16.0548, radius=8.5, center=<-23.4061 0 0>, rotation=<rot 90 0 0>}]
 thigh_r [{shapeType=Capsule, length=16.0548, radius=8.5, center=<23.4061 0 0>, rotation=<rot 90 180 0>}]
 thigh_twist_01_l [{shapeType=Capsule, length=13.9727, radius=8, rotation=<rot 90 0 0>}]
 thigh_twist_01_r [{shapeType=Capsule, length=13.9727, radius=8, rotation=<rot 90 180 0>}]
}
constraints [
 {
  bone1Name=thigh_l
  bone2Name=pelvis
  swing1LimitDegrees=45
  swing1Motion=Locked
  swing2LimitDegrees=45
  swing2Motion=Locked
  twistLimitDegrees=45
  twistMotion=Locked
 }
 {
  bone1Name=foot_l
  bone2Name=thigh_l
  swing1LimitDegrees=45
  swing1Motion=Locked
  swing2LimitDegrees=45
  swing2Motion=Locked
  twistLimitDegrees=45
  twistMotion=Locked
 }
 {
  bone1Name=thigh_twist_01_l
  bone2Name=thigh_l
  swing1LimitDegrees=45
  swing1Motion=Locked
  swing2LimitDegrees=45
  swing2Motion=Locked
  twistLimitDegrees=45
  twistMotion=Locked
 }
 {
  bone1Name=thigh_r
  bone2Name=pelvis
  swing1LimitDegrees=45
  swing1Motion=Locked
  swing2LimitDegrees=45
  swing2Motion=Locked
  twistLimitDegrees=45
  twistMotion=Locked
 }
 {
  bone1Name=calf_r
  bone2Name=thigh_r
  swing1LimitDegrees=45
  swing1Motion=Locked
  swing2LimitDegrees=45
  swing2Motion=Locked
  twistLimitDegrees=45
  twistMotion=Locked
 }
 {
  bone1Name=foot_r
  bone2Name=calf_r
  swing1LimitDegrees=45
  swing1Motion=Locked
  swing2LimitDegrees=45
  swing2Motion=Locked
  twistLimitDegrees=45
  twistMotion=Locked
 }
 {
  bone1Name=thigh_twist_01_r
  bone2Name=thigh_r
  swing1LimitDegrees=45
  swing1Motion=Locked
  swing2LimitDegrees=45
  swing2Motion=Locked
  twistLimitDegrees=45
  twistMotion=Locked
 }
 {
  bone1Name=base_legStrap_00_l
  bone2Name=thigh_l
  swing1LimitDegrees=45
  swing1Motion=Locked
  swing2LimitDegrees=45
  swing2Motion=Locked
  twistLimitDegrees=45
  twistMotion=Locked
 }
 {bone1Name=dyn_legStrap_00_l, bone2Name=base_legStrap_00_l, swing1LimitDegrees=20, swing2LimitDegrees=10, twistLimitDegrees=0}
 {bone1Name=dyn_legStrap_01_l, bone2Name=dyn_legStrap_00_l, swing1LimitDegrees=25, swing2LimitDegrees=15, twistLimitDegrees=0}
]
