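# Physics asset for the male medium StreetBee LowerR cosmetic mesh, simulating leg straps on the right thigh.
# Capsules on the pelvis, left leg and right thigh are joined by locked constraints to form the body collision.
# Front and back strap chains hang from thigh anchors, branching through small-angle swing limits and locked intermediate base bones.
# Cross constraints join front and back branch ends, one free-rotating and one limited, closing the straps into loops.
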PhysicsAsset /BRCosmetics/Characters/Player/Male/Medium/Bodies/M_MED_StreetBee/Meshes/M_MED_StreetBeeLowerR_Physics
bodies {
 base_legStrap_00_br [{shapeType=Capsule, length=0.505, radius=0.505, shapeName=base_legStrap_00_br_capsule, rotation=<rot 0 -90 0>}]
 base_legStrap_00_fr [{shapeType=Capsule, length=0.505, radius=0.505, shapeName=base_legStrap_00_fr_capsule, rotation=<rot 0 -90 0>}]
 base_legStrap_00_fr2 [{shapeType=Capsule, length=0.505, radius=0.505, shapeName=base_legStrap_00_fr2_capsule, rotation=<rot 0 -90 0>}]
 base_legStrap_br_center [{shapeType=Capsule, length=0.505, radius=0.505, shapeName=base_legStrap_br_center_capsule, rotation=<rot 0 -90 0>}]
 base_legStrap_fr__center [{shapeType=Capsule, length=0.505, radius=0.505, shapeName=base_legStrap_fr__center_capsule, rotation=<rot 0 -90 0>}]
 calf_l [{shapeType=Capsule, length=20.0435, radius=8.5, center=<0 -0.8482 0>, rotation=<rot 90 0 0>}]
 dyn_legStrap_00_br [{shapeType=Capsule, length=3.8138, radius=2.7968, shapeName=dyn_legStrap_00_br_capsule, rotation=<rot 90 0 0>}]
 dyn_legStrap_00_br1 [{shapeType=Capsule, length=4.7047, radius=2, shapeName=dyn_legStrap_00_br1_capsule, rotation=<rot 90 0 0>}]
 dyn_legStrap_00_br2 [{shapeType=Capsule, length=4.2991, radius=2, shapeName=dyn_legStrap_00_br2_capsule, rotation=<rot 90 0 0>}]
 dyn_legStrap_00_fr [{shapeType=Capsule, length=0.505, radius=0.505, shapeName=dyn_legStrap_00_fr_capsule, rotation=<rot 90 0 0>}]
 dyn_legStrap_00_fr1 [{shapeType=Capsule, length=4.5346, radius=2, shapeName=dyn_legStrap_00_fr1_capsule, center=<3.0042 0 1.0722>, rotation=<rot 90 0 0>}]
 dyn_legStrap_00_fr2 [{shapeType=Capsule, length=6.3835, radius=1.5, shapeName=dyn_legStrap_00_fr2_capsule, center=<1.158 -1.1537 0>, rotation=<rot 90 0 0>}]
 dyn_legStrap_01_br [{shapeType=Capsule, length=7.7123, radius=2, shapeName=dyn_legStrap_01_br_capsule, center=<-2 0 0>, rotation=<rot 90 0 0>}]
 dyn_legStrap_01_br1 [{shapeType=Capsule, length=5.0922, radius=2, shapeName=dyn_legStrap_01_br1_capsule, rotation=<rot 64.9999 0.0002 0.0003>}]
 dyn_legStrap_01_br2 [{shapeType=Capsule, length=4.3295, radius=2, shapeName=dyn_legStrap_01_br2_capsule, rotation=<rot 90 0 0>}]
 dyn_legStrap_01_fr1 [{
   shapeType=Capsule
   length=4.0371
   radius=2
   shapeName=dyn_legStrap_01_fr1_capsule
   center=<1.2786 0 -0.5962>
   rotation=<rot 65.0001 -0.0003 0.0001>
  }]
 dyn_legStrap_01_fr2 [{shapeType=Capsule, length=7.5371, radius=1.5, shapeName=dyn_legStrap_01_fr1_capsule, center=<-0.7214 0.4889 0>, rotation=<rot 90 24.9999 0>}]
 dyn_legStrap_01_r [{shapeType=Capsule, length=5.4478, radius=2, shapeName=dyn_legStrap_01_r_capsule, rotation=<rot 90 0 0>}]
 foot_l [{shapeType=Capsule, length=15.6916, radius=10, rotation=<rot 90 0 0>}]
 pelvis [{shapeType=Capsule, length=20.3917, radius=9.5, center=<-17.5121 0 0>}]
 thigh_l [{shapeType=Capsule, length=16.0548, radius=8.5, center=<-23.4061 0 0>, rotation=<rot 90 0 0>}]
 thigh_r [{shapeType=Capsule, length=23.5548, radius=9, center=<25.7459 0.8131 -0.9137>, rotation=<rot 85 179.9993 -0.0002>}]
 thigh_twist_01_l [{shapeType=Capsule, length=13.9727, radius=8, rotation=<rot 90 0 0>}]
}
constraints [
 {
  bone1Name=thigh_l
  bone2Name=pelvis
  swing1LimitDegrees=45
  swing1Motion=Locked
  swing2LimitDegrees=45
  swing2Motion=Locked
  twistLimitDegrees=45
  twistMotion=Locked
 }
 {
  bone1Name=calf_l
  bone2Name=thigh_l
  swing1LimitDegrees=45
  swing1Motion=Locked
  swing2LimitDegrees=45
  swing2Motion=Locked
  twistLimitDegrees=45
  twistMotion=Locked
 }
 {
  bone1Name=foot_l
  bone2Name=calf_l
  swing1LimitDegrees=45
  swing1Motion=Locked
  swing2LimitDegrees=45
  swing2Motion=Locked
  twistLimitDegrees=45
  twistMotion=Locked
 }
 {
  bone1Name=thigh_twist_01_l
  bone2Name=thigh_l
  swing1LimitDegrees=45
  swing1Motion=Locked
  swing2LimitDegrees=45
  swing2Motion=Locked
  twistLimitDegrees=45
  twistMotion=Locked
 }
 {
  bone1Name=thigh_r
  bone2Name=pelvis
  swing1LimitDegrees=45
  swing1Motion=Locked
  swing2LimitDegrees=45
  swing2Motion=Locked
  twistLimitDegrees=45
  twistMotion=Locked
 }
 {
  bone1Name=base_legStrap_00_br
  bone2Name=thigh_r
  swing1LimitDegrees=45
  swing1Motion=Locked
  swing2LimitDegrees=45
  swing2Motion=Locked
  twistLimitDegrees=45
  twistMotion=Locked
 }
 {bone1Name=dyn_legStrap_00_br, bone2Name=base_legStrap_00_br, swing1LimitDegrees=5, swing2LimitDegrees=5, twistLimitDegrees=0}
 {bone1Name=dyn_legStrap_01_br, bone2Name=base_legStrap_br_center, swing1LimitDegrees=15, swing2LimitDegrees=15, twistLimitDegrees=0}
 {bone1Name=dyn_legStrap_00_br1, bone2Name=dyn_legStrap_01_br, swing1LimitDegrees=10, swing2LimitDegrees=10, twistLimitDegrees=0}
 {bone1Name=dyn_legStrap_01_br1, bone2Name=dyn_legStrap_00_br1, swing1LimitDegrees=5, swing2LimitDegrees=5, twistLimitDegrees=0}
 {bone1Name=dyn_legStrap_00_br2, bone2Name=dyn_legStrap_01_br, swing1LimitDegrees=10, swing2LimitDegrees=10, twistLimitDegrees=0}
 {bone1Name=dyn_legStrap_01_br2, bone2Name=dyn_legStrap_00_br2, swing1LimitDegrees=25, swing2LimitDegrees=25, twistLimitDegrees=0}
 {
  bone1Name=base_legStrap_00_fr
  bone2Name=thigh_r
  swing1LimitDegrees=45
  swing1Motion=Locked
  swing2LimitDegrees=45
  swing2Motion=Locked
  twistLimitDegrees=45
  twistMotion=Locked
 }
 {bone1Name=dyn_legStrap_00_fr, bone2Name=base_legStrap_00_fr, swing1LimitDegrees=5, swing2LimitDegrees=5, twistLimitDegrees=0}
 {bone1Name=dyn_legStrap_01_r, bone2Name=base_legStrap_fr__center, swing1LimitDegrees=25, swing2LimitDegrees=25, twistLimitDegrees=0}
 {bone1Name=dyn_legStrap_00_fr1, bone2Name=dyn_legStrap_01_r, swing1LimitDegrees=10, swing2LimitDegrees=10, twistLimitDegrees=0}
 {bone1Name=dyn_legStrap_01_fr1, bone2Name=dyn_legStrap_00_fr1, swing1LimitDegrees=5, swing2LimitDegrees=5, twistLimitDegrees=0}
 {bone1Name=dyn_legStrap_00_fr2, bone2Name=base_legStrap_00_fr2, swing1LimitDegrees=20, swing2LimitDegrees=20, twistLimitDegrees=0}
 {bone1Name=dyn_legStrap_01_fr2, bone2Name=dyn_legStrap_00_fr2, swing1LimitDegrees=5, swing2LimitDegrees=5, twistLimitDegrees=0}
 {bone1Name=dyn_legStrap_01_fr2, bone2Name=dyn_legStrap_01_br2, swing1LimitDegrees=25, swing2LimitDegrees=25, twistLimitDegrees=0}
 {
  bone1Name=base_legStrap_00_fr2
  bone2Name=dyn_legStrap_01_r
  swing1LimitDegrees=45
  swing1Motion=Locked
  swing2LimitDegrees=45
  swing2Motion=Locked
  twistLimitDegrees=45
  twistMotion=Locked
 }
 {
  bone1Name=base_legStrap_br_center
  bone2Name=dyn_legStrap_00_br
  swing1LimitDegrees=45
  swing1Motion=Locked
  swing2LimitDegrees=45
  swing2Motion=Locked
  twistLimitDegrees=45
  twistMotion=Locked
 }
 {
  bone1Name=base_legStrap_fr__center
  bone2Name=dyn_legStrap_00_fr
  swing1LimitDegrees=45
  swing1Motion=Locked
  swing2LimitDegrees=45
  swing2Motion=Locked
  twistLimitDegrees=45
  twistMotion=Locked
 }
 {
  bone1Name=dyn_legStrap_01_br1
  bone2Name=dyn_legStrap_01_fr1
  swing1LimitDegrees=45
  swing1Motion=Free
  swing2LimitDegrees=45
  swing2Motion=Free
  twistLimitDegrees=45
  twistMotion=Free
 }
]
