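# Physics asset for the upper-body mesh of the StreetBee medium male player cosmetic, set up for dangling hood and wrist straps.
# Capsules on pelvis, spine and arms are linked by angularly locked constraints, forming a rigid body scaffold.
# Small base capsules lock the straps to the upper spine and forearms.
# From these hang two-segment hood chains with asymmetric tuning and single-segment wrist straps, given narrower swing and zero twist limits.
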PhysicsAsset /BRCosmetics/Characters/Player/Male/Medium/Bodies/M_MED_StreetBee/Meshes/M_MED_StreetBeeUpper_Physics
bodies {
 base_hoodStrap_00_l [{shapeType=Capsule, length=0.505, radius=0.505, shapeName=base_hoodStrap_00_l_capsule, rotation=<rot 0 -90 0>}]
 base_hoodStrap_00_r [{shapeType=Capsule, length=0.505, radius=0.505, shapeName=base_hoodStrap_00_r_capsule, rotation=<rot 0 -90 0>}]
 base_wristStrap_00_l [{shapeType=Capsule, length=0.505, radius=0.505, shapeName=base_wristStrap_00_l_capsule, rotation=<rot 0 -90 0>}]
 base_wristStrap_00_r [{shapeType=Capsule, length=0.505, radius=0.505, shapeName=base_wristStrap_00_r_capsule, rotation=<rot 0 -90 0>}]
 dyn_hoodStrap_00_l [{shapeType=Capsule, length=4.5, radius=0.8, shapeName=dyn_hoodStrap_00_l_capsule, center=<3 0.3 0>, rotation=<rot 90 0 0>}]
 dyn_hoodStrap_00_r [{shapeType=Capsule, length=8, radius=0.8, shapeName=dyn_hoodStrap_00_r_capsule, center=<5 0.4 0>, rotation=<rot 90 0 0>}]
 dyn_hoodStrap_01_l [{shapeType=Capsule, length=5.5, radius=0.8, shapeName=dyn_hoodStrap_01_l_capsule, center=<3 0 0>, rotation=<rot 90 0 0>}]
 dyn_hoodStrap_01_r [{shapeType=Capsule, length=4, radius=0.8, shapeName=dyn_hoodStrap_01_r_capsule, center=<2.5 0 0>, rotation=<rot 90 0 0>}]
 dyn_wristStrap_00_l [{shapeType=Capsule, length=2.5, radius=2, shapeName=dyn_wristStrap_00_l_capsule, center=<2.4 0 0>, rotation=<rot 90 0 0>}]
 dyn_wristStrap_00_r [{shapeType=Capsule, length=2.5, radius=2, shapeName=dyn_wristStrap_00_r_capsule, center=<-2.4 0 0>, rotation=<rot 90 0 0>}]
 hand_l [{shapeType=Capsule, length=9.098, radius=10, rotation=<rot 90 0 0>}]
 hand_r [{shapeType=Capsule, length=9.2958, radius=10, rotation=<rot 90 0 0>}]
 lowerarm_l [{shapeType=Capsule, length=14.4988, radius=10, rotation=<rot 90 0 0>}]
 lowerarm_r [{shapeType=Capsule, length=13.387, radius=10, rotation=<rot 90 0 0>}]
 pelvis [{shapeType=Capsule, length=19.8917, radius=10}]
 spine_02 [{shapeType=Capsule, length=19.2513, radius=10}]
 spine_03 [{shapeType=Capsule, length=27.1696, radius=10, center=<0 1.6994 0>}]
 spine_04 [{shapeType=Capsule, length=17.0105, radius=11.8248, shapeName=spine_04_capsule, rotation=<rot 0 -0.0001 0>}]
 spine_05 [{shapeType=Capsule, length=19.1682, radius=10.5}]
 upperarm_l [{shapeType=Capsule, length=15.9446, radius=10, rotation=<rot 90 0 0>}]
 upperarm_r [{shapeType=Capsule, length=19.3028, radius=7, center=<-13.4513 0 0.2937>, rotation=<rot 90 0 0>}]
}
constraints [
 {
  bone1Name=spine_02
  bone2Name=pelvis
  swing1LimitDegrees=45
  swing1Motion=Locked
  swing2LimitDegrees=45
  swing2Motion=Locked
  twistLimitDegrees=45
  twistMotion=Locked
 }
 {
  bone1Name=spine_03
  bone2Name=spine_02
  swing1LimitDegrees=45
  swing1Motion=Locked
  swing2LimitDegrees=45
  swing2Motion=Locked
  twistLimitDegrees=45
  twistMotion=Locked
 }
 {
  bone1Name=spine_05
  bone2Name=spine_04
  swing1LimitDegrees=45
  swing1Motion=Locked
  swing2LimitDegrees=45
  swing2Motion=Locked
  twistLimitDegrees=45
  twistMotion=Locked
 }
 {
  bone1Name=upperarm_l
  bone2Name=spine_05
  swing1LimitDegrees=45
  swing1Motion=Locked
  swing2LimitDegrees=45
  swing2Motion=Locked
  twistLimitDegrees=45
  twistMotion=Locked
 }
 {
  bone1Name=lowerarm_l
  bone2Name=upperarm_l
  swing1LimitDegrees=45
  swing1Motion=Locked
  swing2LimitDegrees=45
  swing2Motion=Locked
  twistLimitDegrees=45
  twistMotion=Locked
 }
 {
  bone1Name=hand_l
  bone2Name=lowerarm_l
  swing1LimitDegrees=45
  swing1Motion=Locked
  swing2LimitDegrees=45
  swing2Motion=Locked
  twistLimitDegrees=45
  twistMotion=Locked
 }
 {
  bone1Name=upperarm_r
  bone2Name=spine_05
  swing1LimitDegrees=45
  swing1Motion=Locked
  swing2LimitDegrees=45
  swing2Motion=Locked
  twistLimitDegrees=45
  twistMotion=Locked
 }
 {
  bone1Name=lowerarm_r
  bone2Name=upperarm_r
  swing1LimitDegrees=45
  swing1Motion=Locked
  swing2LimitDegrees=45
  swing2Motion=Locked
  twistLimitDegrees=45
  twistMotion=Locked
 }
 {
  bone1Name=hand_r
  bone2Name=lowerarm_r
  swing1LimitDegrees=45
  swing1Motion=Locked
  swing2LimitDegrees=45
  swing2Motion=Locked
  twistLimitDegrees=45
  twistMotion=Locked
 }
 {
  bone1Name=base_hoodStrap_00_l
  bone2Name=spine_05
  swing1LimitDegrees=45
  swing1Motion=Locked
  swing2LimitDegrees=45
  swing2Motion=Locked
  twistLimitDegrees=45
  twistMotion=Locked
 }
 {bone1Name=dyn_hoodStrap_00_l, bone2Name=base_hoodStrap_00_l, swing1LimitDegrees=25, swing2LimitDegrees=10, twistLimitDegrees=0}
 {bone1Name=dyn_hoodStrap_01_l, bone2Name=dyn_hoodStrap_00_l, swing1LimitDegrees=30, swing2LimitDegrees=20, twistLimitDegrees=0}
 {
  bone1Name=base_hoodStrap_00_r
  bone2Name=spine_05
  swing1LimitDegrees=45
  swing1Motion=Locked
  swing2LimitDegrees=45
  swing2Motion=Locked
  twistLimitDegrees=45
  twistMotion=Locked
 }
 {bone1Name=dyn_hoodStrap_00_r, bone2Name=base_hoodStrap_00_r, swing1LimitDegrees=30, swing2LimitDegrees=15, twistLimitDegrees=0}
 {bone1Name=dyn_hoodStrap_01_r, bone2Name=dyn_hoodStrap_00_r, swing1LimitDegrees=45, swing2LimitDegrees=20, twistLimitDegrees=0}
 {
  bone1Name=base_wristStrap_00_l
  bone2Name=lowerarm_l
  swing1LimitDegrees=45
  swing1Motion=Locked
  swing2LimitDegrees=45
  swing2Motion=Locked
  twistLimitDegrees=45
  twistMotion=Locked
 }
 {bone1Name=dyn_wristStrap_00_l, bone2Name=base_wristStrap_00_l, swing1LimitDegrees=0, swing2LimitDegrees=15, twistLimitDegrees=0}
 {
  bone1Name=base_wristStrap_00_r
  bone2Name=lowerarm_r
  swing1LimitDegrees=45
  swing1Motion=Locked
  swing2LimitDegrees=45
  swing2Motion=Locked
  twistLimitDegrees=45
  twistMotion=Locked
 }
 {bone1Name=dyn_wristStrap_00_r, bone2Name=base_wristStrap_00_r, swing1LimitDegrees=0, swing2LimitDegrees=15, twistLimitDegrees=0}
 {
  bone1Name=spine_04
  bone2Name=spine_03
  swing1LimitDegrees=45
  swing1Motion=Locked
  swing2LimitDegrees=45
  swing2Motion=Locked
  twistLimitDegrees=45
  twistMotion=Locked
 }
]
