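# Post-process-style cosmetic AnimBP for the M_MED_StreetBee body, adding secondary motion to a pose copied from the parent mesh.
# A state-driven Control Rig feeds hood, upper and per-leg RigidBody sims, wrist-strap AnimDynamics, then cape/backpack attach offsets.
# Back bling disables hood physics.
# The EventGraph sets rig alphas and gravity/sim-space variables, and drives per-foot material WPO cloth outside the front end, with stronger wind while skydiving.
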AnimBlueprint /BRCosmetics/Characters/Player/Male/Medium/Bodies/M_MED_StreetBee/Meshes/M_MED_StreetBee_AnimBP
target_skeleton=&/M_MED_StreetBee_Skeleton
graphs [
 AnimGraph {nodes [
   Anim Dynamics - Bone: anim_hood_bt!(ComponentPose, bAlphaBoolEnabled:bool=True, ExternalForce="0,0,0") -> (Pose) {
    alphaInputType=Bool
    angularDampingOverride=0.9999
    bGravityOverrideInSimSpace=true
    bOverrideAngularDamping=true
    bUseGravityOverride=true
    simulationSpace=BoneRelative
    alphaBoolBlend {blendOption=Linear}
    boundBone {boneName=anim_hood_bt}
    componentAppliedLinearAccClamp=<3000 3000 3000>
    componentLinearAccScale=<0.3 0.3 0.3>
    componentLinearVelScale=<0.3 0.3 0.3>
    gravityOverride=<-980 100>
    physicsBodyDefinitions [{
      boundBone {boneName=anim_hood_bt}
      boxExtents=<10 20 8>
      constraintSetup {angularLimitsMax=<0 3 20>, angularLimitsMin=<0 -30 0>, angularTarget=<0 0 0>}
      localJointOffset=<1 0 0>
     }]
    relativeSpaceBone {boneName=root}
   }
   n1 = Control Rig!(
    Source=Copy Pose From Mesh() {bUseAttachedParent=true}
    Alpha
    crouch_alpha:double=Get crouch_alpha() -> (crouch_alpha:double)
    moving2D_alpha:double=Get moving2D_alpha() -> (moving2D_alpha:double)
    aim_alpha:double=Get aim_alpha() -> (aim_alpha:double)
    idle_alpha:double=Get idle_alpha() -> (idle_alpha:double)
    is on grind rail:bool=Get bIsOnGrindRail()
    RBAN_alpha:double=Get rigidbody_alpha() -> (rigidbody_alpha:double)
    isFrontEnd:bool=Get bIsInFrontEnd()
    baseballAlpha:double=Get baseballSlide_Alpha() -> (baseballSlide_Alpha:double)
    DBNO_Alpha:bool=Get bIsDBNO()
   ) -> (Pose) {
    alphaBoolBlend {blendOption=Linear}
    controlRigAssetReference {blueprintRigClass=&/M_MED_StreetBee_CtrlRig.M_MED_StreetBee_CtrlRig_C}
   }
   n2 = RigidBody!(
    ComponentPose=Local To Component(LocalPose=@n1)
    bAlphaBoolEnabled:bool=NOT Boolean(A=Get IsWearingBackbling() -> (IsWearingBackbling:bool))
    OverrideWorldGravity:Vector="0,0,0"
    ExternalForce
   ) -> (Pose) {
    alphaInputType=Bool
    bOverrideWorldGravity=true
    overridePhysicsAsset=&/M_MED_StreetBeeHood_Physics1
    simulationSpace=BaseBoneSpace
    alphaBoolBlend {blendOption=Linear}
    baseBoneRef {boneName=spine_05}
    simSpaceSettings {
     maxAngularAcceleration=1000
     maxAngularVelocity=3000
     maxLinearAcceleration=3000
     maxLinearVelocity=1000
     velocityScaleZ=0.7
     worldAlpha=1
     externalLinearDragV=<0.3 0.2 0.2>
     externalLinearVelocity=<0.3 0.3 0.3>
    }
   }
   n3 = RigidBody!(ComponentPose=@n2, Alpha:float, OverrideWorldGravity="0,0,0", ExternalForce) -> (Pose) {
    bOverrideWorldGravity=true
    overridePhysicsAsset=&/M_MED_StreetBeeUpper_Physics
    simulationSpace=BaseBoneSpace
    alphaBoolBlend {blendOption=Linear}
    baseBoneRef {boneName=spine_05}
    simSpaceSettings {
     maxAngularAcceleration=1000
     maxAngularVelocity=3000
     maxLinearAcceleration=3000
     maxLinearVelocity=1000
     velocityScaleZ=0.7
     worldAlpha=1
     externalLinearDragV=<0.3 0.2 0.2>
     externalLinearVelocity=<0.3 0.3 0.3>
    }
   }
   n4 = RigidBody(ComponentPose=@n3, OverrideWorldGravity="0,0,0") {
    bOverrideWorldGravity=true
    overridePhysicsAsset=&/M_MED_StreetBeeLowerL_Physics
    simulationSpace=BaseBoneSpace
    alphaBoolBlend {blendOption=Linear}
    baseBoneRef {boneName=thigh_l}
    simSpaceSettings {
     maxAngularAcceleration=1000
     maxAngularVelocity=3000
     maxLinearAcceleration=3000
     maxLinearVelocity=1000
     velocityScaleZ=0.5
     worldAlpha=1
     externalAngularVelocity=<1 1 1>
     externalLinearDragV=<0.3 0.2 0.2>
     externalLinearVelocity=<0.3 0.3 0.3>
    }
   }
   n5 = RigidBody(ComponentPose=@n4, OverrideWorldGravity="0,0,0") {
    bOverrideWorldGravity=true
    overridePhysicsAsset=&/M_MED_StreetBeeLowerR_Physics
    simulationSpace=BaseBoneSpace
    alphaBoolBlend {blendOption=Linear}
    baseBoneRef {boneName=thigh_r}
    simSpaceSettings {
     maxAngularAcceleration=1000
     maxAngularVelocity=3000
     maxLinearAcceleration=3000
     maxLinearVelocity=1000
     worldAlpha=1
     externalAngularVelocity=<1 1 1>
     externalLinearDragV=<0.3 0.2 0.2>
     externalLinearVelocity=<0.3 0.3 0.3>
    }
   }
   n6 = Anim Dynamics - Bone: dyn_wristStrap_00_l(ComponentPose=@n5, Alpha=1, ExternalForce="0,0,0") {
    angularDampingOverride=1
    angularSpringConstant=20000
    bAngularSpring=true
    bOverrideAngularDamping=true
    bUseGravityOverride=true
    gravityScale=0.35
    simulationSpace=BoneRelative
    alphaBoolBlend {blendOption=Linear}
    boundBone {boneName=dyn_wristStrap_00_l}
    componentAppliedLinearAccClamp=<3000 3000 3000>
    componentLinearAccScale=<0.5 0.5 0.5>
    componentLinearVelScale=<0.5 0.5 0.5>
    gravityOverride {z=-980}
    physicsBodyDefinitions [{
      boundBone {boneName=dyn_wristStrap_00_l}
      boxExtents=<20 30 30>
      constraintSetup {angularLimitsMax=<15 40 10>, angularLimitsMin=<-15 -40 -10>, angularTarget=<1 0 0>}
      localJointOffset=<4 0 0>
     }]
    relativeSpaceBone {boneName=root}
   }
   n7 = Anim Dynamics - Bone: dyn_wristStrap_00_r(ComponentPose=@n6, Alpha=1, ExternalForce="0,0,0") {
    angularDampingOverride=1
    angularSpringConstant=20000
    bAngularSpring=true
    bOverrideAngularDamping=true
    bUseGravityOverride=true
    gravityScale=0.35
    simulationSpace=BoneRelative
    alphaBoolBlend {blendOption=Linear}
    boundBone {boneName=dyn_wristStrap_00_r}
    componentAppliedLinearAccClamp=<3000 3000 3000>
    componentLinearAccScale=<0.5 0.5 0.5>
    componentLinearVelScale=<0.5 0.5 0.5>
    gravityOverride {z=-980}
    physicsBodyDefinitions [{
      boundBone {boneName=dyn_wristStrap_00_r}
      boxExtents=<20 30 30>
      constraintSetup {angularLimitsMax=<10 30 15>, angularLimitsMin=<-10 -30 -15>, angularTarget=<1 0 0>}
      localJointOffset=<4 0 0>
     }]
    relativeSpaceBone {boneName=root}
   }
   n8 = "Transform (Modify) Bone - Bone: attach_cape"(ComponentPose=@n7, Alpha=1, Translation="-0.000,-3.229,0.260", Rotation="0,0,0", Scale="1,1,1") {
    translationMode=BMM_Additive
    translationSpace=BCS_BoneSpace
    alphaBoolBlend {blendOption=Linear}
    boneToModify {boneName=attach_cape}
    translation {y=-3.229, z=0.26}
   }
   n9 = "Transform (Modify) Bone - Bone: attach_backpack"(ComponentPose=@n8, Alpha=1, Translation="0,-4.25,0", Rotation="0,0,0", Scale="1,1,1") {
    translationMode=BMM_Additive
    translationSpace=BCS_BoneSpace
    alphaBoolBlend {blendOption=Linear}
    boneToModify {boneName=attach_backpack}
    translation {y=-4.25}
   }
   Output Pose(Result=Component To Local(ComponentPose=@n9)) {layerGroup=None}
  ]}
 EventGraph {nodes [
   n1 = Event BlueprintUpdateAnimation()
   > n2 = Sequence!(execute) -> (then_0, then_1, then_2:Exec, then_3:Exec, then_4:Exec, then_5:Exec, then_6:Exec)
   > n3 = Gravity Override State Function DEPRECATED(
    CharacterPartAnimInstance=Self-Reference()
    GameTimeInSeconds=GetGameTimeInSeconds()
    GravityOverrideParams=Get Gravity Override Params()
    RBAN Params=Get RBAN Params()
   )
   > Set GravityOverride(GravityOverride=@n3)
   > Set Sim Space Settings(Sim Space Settings:Sim Space Settings Structure=Break RBANSetting Struct!(RBANSettingStruct:"RBANSetting Struct Structure (by ref)"=@n3.RBAN SettingsOut) -> (
     ComponentLinearAccScale_28_#1:Vector=~
     ComponentLinearVelScale_31_#2:Vector=~
     SimSpaceSettings_34_#3:Sim Space Settings Structure
    ) {is=title K2Node_BreakStruct}.SimSpaceSettings_34_#3) -> (Output_Get:Sim Space Settings Structure)
   Is Valid(exec=@n2.then_1, InputObject=Get PlayerPawn() -> (PlayerPawn:Fort Player Pawn Athena Object Reference))
   > n4 = Branch(Condition=Get bIsInFrontEnd())
   Set WPO Globals Function(
    execute=@n4.else
    Material=Get Mat Instance Dynamic()
    Overlay Material=Get Overlay Mat Instance Dynamic()
    Skeletal Mesh=Get SkelMesh()
    "Local Noise Transform  Location"=Get Local Noise Transform Location() -> (Local Noise Transform Location:Vector)
    Local Noise Offset=Get Local Noise Offset()
   )
   > Set WPO Regional Function!(
    execute
    Material:Material Instance Dynamic Object Reference=Get Mat Instance Dynamic()
    Overlay Material:Material Instance Dynamic Object Reference=Get Overlay Mat Instance Dynamic()
    Skeletal Mesh:Skeletal Mesh Component Object Reference=Get SkelMesh()
    Velocity Transform:Name=foot_r
    Velocity Interp Speed:double=Get Velocity Interp Speed()
    Collision Transform:Name=calf_r
    Collision Transform Axis:EAxis Enum
    Invert Collision Transform Axis:bool=false
    Delta Time X:double=@n1.DeltaTimeX
    Previous Velocity Transform Location:"Vector (by ref)"=Get R1 Velocity Transform Location()
    Previous Velocity Transform Velocity:"Vector (by ref)"=Get R1 Velocity Transform Velocity()
    Region:int=1
   ) -> (then)
   > Set WPO Regional Function(
    Material=Get Mat Instance Dynamic()
    Overlay Material=Get Overlay Mat Instance Dynamic()
    Skeletal Mesh=Get SkelMesh()
    Velocity Transform=foot_l
    Velocity Interp Speed=Get Velocity Interp Speed()
    Collision Transform=calf_l
    Invert Collision Transform Axis=true
    Delta Time X=@n1.DeltaTimeX
    Previous Velocity Transform Location=Get R2 Velocity Transform Location()
    Previous Velocity Transform Velocity=Get R2 Velocity Transform Velocity()
    Region=2
   )
   > n5 = Branch(Condition=Get bIsSkyDiving())
   > Set WPO Pho Wind Intensity(WPO Pho Wind Intensity=0.8)
   > n6 = Set WPO Pho Turbulence(WPO Pho Turbulence:double=0.8) -> (Output_Get:double)
   SetScalarParameterValue(
    execute=[@n6.then, @n8.then]
    self=[Get Overlay Mat Instance Dynamic(), Get Mat Instance Dynamic()]
    ParameterName=ClothSim_WindStrength
    Value=Get WPO Pho Wind Intensity()
   )
   n7 = Get WPO Pho Turbulence() -> (WPO Pho Turbulence:double)
   > SetScalarParameterValue(self=[Get Overlay Mat Instance Dynamic(), Get Mat Instance Dynamic()], ParameterName=ClothSim_R1_TurbulenceStrength, Value=@n7)
   > SetScalarParameterValue(self=[Get Mat Instance Dynamic(), Get Overlay Mat Instance Dynamic()], ParameterName=ClothSim_R2_TurbulenceStrength, Value=@n7)
   Set WPO Pho Wind Intensity(execute=@n5.else, WPO Pho Wind Intensity=0.1)
   > n8 = Set WPO Pho Turbulence(WPO Pho Turbulence=0.2)
   n9 = Branch(execute=@n2.then_5, Condition=Get bRigidBodyEnabled())
   > Set rigidbody_alpha(rigidbody_alpha:double=1) -> (Output_Get:double)
   Set rigidbody_alpha(execute=@n9.else, rigidbody_alpha=0)
   n10 = Get bIsMoving2D()
   n11 = Get bIsPlayingEmote()
   Set isIDLE(execute=@n2.then_3, isIDLE:bool=NOT Boolean(A=OR Boolean!(A=@n10, B=@n11, C:bool) -> (ReturnValue))) -> (Output_Get:bool)
   n12 = Branch(execute=@n2.then_2, Condition=Get bIsMoving2D())
   > Set moving2D_alpha(moving2D_alpha:double=1) -> (Output_Get:double)
   Set moving2D_alpha(execute=@n12.else, moving2D_alpha=0)
   n13 = Branch(execute=@n2.then_4, Condition=Get isIDLE() -> (isIDLE:bool))
   > Set idle_alpha(idle_alpha:double=1) -> (Output_Get:double)
   Set idle_alpha(execute=@n13.else, idle_alpha=0)
   Event OnBeginBaseballSliding() {is=title K2Node_Event}
   > Set baseballSlide_Alpha(baseballSlide_Alpha:double=1) -> (Output_Get:double)
   Event OnCharacterCustomizationLoaded()
   > n14 = Sequence!(execute) -> (then_0, then_1)
   > n15 = Set SkelMesh(SkelMesh=GetPartSkeletalMeshComponent())
   > n16 = Get Dynamic Material Instance(SkeletalMeshComponent=@n15)
   > Set Mat Instance Dynamic(Mat Instance Dynamic=@n16)
   > n17 = GetOrCreateDynamicOverlayMaterial(MeshComponent=@n15, MaterialName=None)
   > Set Overlay Mat Instance Dynamic(Overlay Mat Instance Dynamic=@n17)
   n18 = CheckForBackbling()
   Set IsWearingBackbling(execute=@n14.then_1, IsWearingBackbling:bool=OR Boolean(A=@n18, B=@n18.bIsBackblingACape, C=@n18.bIsBackblingLong)) -> (Output_Get:bool)
   Event OnBeginCrouching()
   > Set crouch_alpha(crouch_alpha:double=1) -> (Output_Get:double)
   Event OnBeginTargeting() {is=title K2Node_Event}
   > Set aim_alpha(aim_alpha:double=1) -> (Output_Get:double)
   Event OnEndTargeting() {is=title K2Node_Event}
   > Set aim_alpha(aim_alpha=0)
   Event BlueprintInitializeAnimation()
   > Set SkelMesh(SkelMesh=GetOwnerSkeletalMeshComponent())
   n19 = TryGetPawnOwner()
   > Is Valid(InputObject=@n19)
   > n20 = Cast To FortPlayerPawnAthena(Object=@n19)
   > Set PlayerPawn(PlayerPawn:Fort Player Pawn Athena Object Reference=@n20) -> (Output_Get:Fort Player Pawn Athena Object Reference)
   Event OnEndBaseballSliding() {is=title K2Node_Event}
   > Set baseballSlide_Alpha(baseballSlide_Alpha=0)
   Event OnEndCrouching()
   > Set crouch_alpha(crouch_alpha=0)
  ]}
]
variables [
 aim_alpha {type=number}
 baseballSlide_Alpha {type=number}
 crouch_alpha {type=number}
 DBNO_invAlpha {type=number}
 Gravity Override Params {type=GravityOverrideParamsStruct}
 GravityOverride {type=Vector}
 idle_alpha {type=number}
 isIDLE {type=boolean}
 IsWearingBackbling {type=boolean}
 Local Noise Offset {type=Vector}
 Local Noise Transform Location {type=Vector}
 Mat Instance Dynamic {type=MaterialInstanceDynamic}
 moving2D_alpha {type=number}
 Overlay Mat Instance Dynamic {type=MaterialInstanceDynamic}
 PlayerPawn {type=FortPlayerPawnAthena}
 R1 Velocity Transform Location {type=Vector}
 R1 Velocity Transform Velocity {type=Vector}
 R2 Velocity Transform Location {type=Vector}
 R2 Velocity Transform Velocity {type=Vector}
 RBAN Params {
  type=RBANParamsStruct
  default {
   dBNO {simSpaceSettings {
     maxAngularAcceleration=1000
     maxAngularVelocity=3
     maxLinearAcceleration=300
     maxLinearVelocity=700
     velocityScaleZ=0
     externalLinearDragV=<0 0.5 0>
    }}
   dBNO_Carried {simSpaceSettings {
     maxAngularAcceleration=1000
     maxAngularVelocity=3
     maxLinearAcceleration=300
     maxLinearVelocity=700
     velocityScaleZ=0
     externalLinearDragV=<0 0.5 0>
    }}
   emoteOrMelee {simSpaceSettings {
     maxAngularAcceleration=1000
     maxAngularVelocity=3
     maxLinearAcceleration=300
     maxLinearVelocity=700
     velocityScaleZ=0
     externalLinearDragV=<0 0.5 0>
    }}
   front_End {simSpaceSettings {
     maxAngularAcceleration=1000
     maxAngularVelocity=3
     maxLinearAcceleration=300
     maxLinearVelocity=700
     velocityScaleZ=0
     externalLinearDragV=<0 0.5 0>
    }}
   onGround_Crouch_Moving {simSpaceSettings {
     maxAngularAcceleration=1000
     maxAngularVelocity=3
     maxLinearAcceleration=300
     maxLinearVelocity=700
     velocityScaleZ=0
     externalLinearDragV=<0 0.5 0>
    }}
   onGround_Falling {simSpaceSettings {
     maxAngularAcceleration=1000
     maxAngularVelocity=3
     maxLinearAcceleration=300
     maxLinearVelocity=700
     velocityScaleZ=0
     externalLinearDragV=<0 0.5 0>
    }}
   onGround_Moving {simSpaceSettings {
     maxAngularAcceleration=1000
     maxAngularVelocity=3
     maxLinearAcceleration=300
     maxLinearVelocity=700
     velocityScaleZ=0
     externalLinearDragV=<0 0.5 0>
    }}
   onGround_Standing {simSpaceSettings {
     maxAngularAcceleration=1000
     maxAngularVelocity=3
     maxLinearAcceleration=300
     maxLinearVelocity=700
     velocityScaleZ=0
     externalLinearDragV=<0 0.5 0>
    }}
   rail_Grinding {simSpaceSettings {
     maxAngularAcceleration=1000
     maxAngularVelocity=3
     maxLinearAcceleration=300
     maxLinearVelocity=700
     velocityScaleZ=0
     worldAlpha=0.1
     externalLinearDragV=<0 0.5 0>
    }}
   riding_Moving {simSpaceSettings {
     maxAngularAcceleration=1000
     maxAngularVelocity=3
     maxLinearAcceleration=300
     maxLinearVelocity=700
     velocityScaleZ=0
     externalLinearDragV=<0 0.5 0>
    }}
   riding_Moving_Targeting {simSpaceSettings {
     maxAngularAcceleration=1000
     maxAngularVelocity=3
     maxLinearAcceleration=300
     maxLinearVelocity=700
     velocityScaleZ=0
     externalLinearDragV=<0 0.5 0>
    }}
   riding_Standing {simSpaceSettings {
     maxAngularAcceleration=1000
     maxAngularVelocity=3
     maxLinearAcceleration=300
     maxLinearVelocity=700
     velocityScaleZ=0
     externalLinearDragV=<0 0.5 0>
    }}
   skydive_DivingDown {simSpaceSettings {
     maxAngularAcceleration=1000
     maxAngularVelocity=3
     maxLinearAcceleration=300
     maxLinearVelocity=700
     velocityScaleZ=0
     externalLinearDragV=<0 0.5 0>
    }}
   skydive_FloatingDown {
    componentLinearAccScale=<0 0.5 0>
    componentLinearVelScale=<0 0.5 0>
    simSpaceSettings {
     maxAngularAcceleration=1000
     maxAngularVelocity=3
     maxLinearAcceleration=300
     maxLinearVelocity=700
     velocityScaleZ=0
     externalLinearDragV=<0 0.5 0>
    }
   }
   skydive_GoingUp {simSpaceSettings {
     maxAngularAcceleration=1000
     maxAngularVelocity=3
     maxLinearAcceleration=300
     maxLinearVelocity=700
     velocityScaleZ=0
     externalLinearDragV=<0 0.5 0>
    }}
   skydive_Parachuting {
    componentLinearAccScale=<0 0.1 0.3>
    componentLinearVelScale=<0 0.1 0.3>
    simSpaceSettings {
     maxAngularAcceleration=1000
     maxAngularVelocity=3
     maxLinearAcceleration=300
     maxLinearVelocity=700
     velocityScaleZ=0
     externalLinearDragV=<0 0.5 0>
    }
   }
   skydive_Surfing {simSpaceSettings {
     maxAngularAcceleration=1000
     maxAngularVelocity=3
     maxLinearAcceleration=300
     maxLinearVelocity=700
     velocityScaleZ=0
     externalLinearDragV=<0 0.5 0>
    }}
   swimming {simSpaceSettings {
     maxAngularAcceleration=1000
     maxAngularVelocity=3
     maxLinearAcceleration=300
     maxLinearVelocity=700
     velocityScaleZ=0
     externalLinearDragV=<0 0.5 0>
    }}
   tactical_Sprint {simSpaceSettings {
     maxAngularAcceleration=1000
     maxAngularVelocity=3
     maxLinearAcceleration=300
     maxLinearVelocity=700
     velocityScaleZ=0
     externalLinearDragV=<0 0.5 0>
    }}
   ziplining {simSpaceSettings {
     maxAngularAcceleration=1000
     maxAngularVelocity=3
     maxLinearAcceleration=300
     maxLinearVelocity=700
     velocityScaleZ=0
     externalLinearDragV=<0 0.5 0>
    }}
  }
 }
 rigidbody_alpha {type=number}
 Sim Space Settings {
  type=SimSpaceSettings
  default {
   dampingAlpha=1
   maxAngularAcceleration=10000
   maxAngularVelocity=10000
   maxLinearAcceleration=10000
   maxLinearVelocity=10000
   velocityScaleZ=1
   worldAlpha=0
   externalLinearDragV=<0 0 0>
  }
 }
 SkelMesh {type=SkeletalMeshComponent}
 Velocity Interp Speed {type=number}
 WPO Pho Turbulence {type=number}
 WPO Pho Wind Intensity {type=number}
]
