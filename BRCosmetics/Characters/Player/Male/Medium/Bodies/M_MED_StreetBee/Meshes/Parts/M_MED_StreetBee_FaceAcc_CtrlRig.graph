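# Control Rig for the StreetBee face accessory that rotates antenna base bones, including the lower pair, as the head bends.
# DeltaRotator converts one swing axis of the head, relative to a neck-parented null, into a clamped weight for ModifyTransforms.
# When IsCrouching is set, a curve reshapes that weight.
# Neck and head controls drive their bones only with the Controllers Debug flag, off here; DeltaRotation and AlphaInterp are unused.
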ControlRigBlueprint /BRCosmetics/Characters/Player/Male/Medium/Bodies/M_MED_StreetBee/Meshes/Parts/M_MED_StreetBee_FaceAcc_CtrlRig
graphs [
 New Function_ContainedGraph {nodes [
   n1 = RigVMFunctionEntryNode()
   RigVMFunction_MathBoolToFloat = &RigVMFunction_MathBoolToFloat(Value=@n1.Debug)
   > &RigUnit_SetTransform(
    Item=(Type=Bone,Name="neck_01")
    bInitial=False
    Value=&RigUnit_GetTransform(Item=(Type=Control,Name="neck_01_ctrl"), bInitial=False).Transform
    Weight=@RigVMFunction_MathBoolToFloat.Result
   )
   > &RigUnit_SetTransform(
    Item=(Type=Bone,Name="neck_02")
    bInitial=False
    Value=&RigUnit_GetTransform(Item=(Type=Control,Name="neck_02_ctrl"), bInitial=False).Transform
    Weight=@RigVMFunction_MathBoolToFloat.Result
   )
   RigUnit_GetTransform = &RigUnit_GetTransform(Item=(Type=Control,Name="head_ctrl"), bInitial=False)
   > &RigUnit_SetTransform(Item=(Type=Bone,Name="head"), bInitial=False, Value=@RigUnit_GetTransform.Transform, Weight=@RigVMFunction_MathBoolToFloat.Result)
   > RigVMFunctionReturnNode()
  ]}
 New Function_ContainedGraph {nodes [
   n1 = RigVMFunctionEntryNode()
   RigVMFunction_MathQuaternionToEuler = &RigVMFunction_MathQuaternionToEuler(
    RotationOrder=ZYX
    Value=&RigUnit_GetRelativeTransformForItem(
     Child=(Type=Bone,Name="head")
     Parent=(Type=Bone,Name="neck_01")
     bChildInitial=False
     bParentInitial=False
     Child.Name=@n1.Child
     Parent.Name=@n1.Parent
    ).RelativeTransform.Rotation
   )
   Remap = &RigVMFunction_MathDoubleRemap(
    TargetMaximum=1
    TargetMinimum=0
    SourceMaximum=@n1.Source Max
    SourceMinimum=@n1.Source Min
    Value=@RigVMFunction_MathQuaternionToEuler.Result.Z
   )
   > RigVMFunctionReturnNode(
    Delta_z=@Remap.Result
    Delta_X_Negative=&RigVMFunction_MathDoubleRemap(SourceMaximum=-30, SourceMinimum=-20, TargetMaximum=1, TargetMinimum=0, Value=@RigVMFunction_MathQuaternionToEuler.Result.X).Result
    Delta_X_Positive=&RigVMFunction_MathDoubleRemap(SourceMaximum=60, SourceMinimum=0, TargetMaximum=1, TargetMinimum=0, Value=@RigVMFunction_MathQuaternionToEuler.Result.X).Result
   )
  ]}
 New Function_ContainedGraph {nodes [
   n1 = RigVMFunctionEntryNode()
   RigVMFunction_MathBoolToInteger = &RigVMFunction_MathBoolToInteger(Value=@n1.onlyTwist)
   RigVMFunction_MathQuaternionSwingTwist = &RigVMFunction_MathQuaternionSwingTwist(
    twistAxis=@n1.twistAxis
    Input=&RigUnit_GetRelativeTransformForItem(
     bChildInitial=False
     bParentInitial=False
     child=(Type=Bone,Name="None")
     parent=(Type=Bone,Name="None")
     child.Name=@n1.child
     parent.Name=@n1.parent
     parent.Type=@n1.parentType
    ).RelativeTransform.Rotation
   )
   RigVMFunction_MathQuaternionToEuler = &RigVMFunction_MathQuaternionToEuler(
    Value=RigVMDispatchNode(
     Values=((),())
     Index=@RigVMFunction_MathBoolToInteger.Result
     Values.0=@RigVMFunction_MathQuaternionSwingTwist.Swing
     Values.1=@RigVMFunction_MathQuaternionSwingTwist.Twist
    ) {template="DISPATCH_RigVMDispatch_SelectInt32(in Index,in Values,out Result)"}.Result
    RotationOrder=@n1.rotationOrder
   )
   > DISPATCH_RigVMDispatch_Print = RigVMDispatchNode(ScreenColor=(R=1,G=1,B=1,A=1), ScreenDuration=1.5, Enabled=@n1.debugRotation, Value=@RigVMFunction_MathQuaternionToEuler.Result) {template="DISPATCH_RigVMDispatch_Print(in Prefix,in Value,in Enabled,in ScreenDuration,in ScreenColor)"}
   Clamp = &RigVMFunction_MathIntClamp(Value=@n1.primAxis)
   DISPATCH_RigVMDispatch_SelectInt32 = RigVMDispatchNode(
    Values=(0,0,0)
    Index=@Clamp.Result
    Values.0=@RigVMFunction_MathQuaternionToEuler.Result.X
    Values.1=@RigVMFunction_MathQuaternionToEuler.Result.Y
    Values.2=@RigVMFunction_MathQuaternionToEuler.Result.Z
   ) {template="DISPATCH_RigVMDispatch_SelectInt32(in Index,in Values,out Result)"}
   Remap = &RigVMFunction_MathDoubleRemap(
    TargetMaximum=1
    TargetMinimum=0
    bClamp=False
    Value=@DISPATCH_RigVMDispatch_SelectInt32.Result
    SourceMaximum=@n1.max
    SourceMinimum=@n1.min
   )
   Multiply = &RigVMFunction_MathDoubleMul(A=&RigVMFunction_MathDoubleClamp(Value=@Remap.Result).Result, B=@n1.premult)
   > RigVMFunctionReturnNode(outputWeight=@Multiply.Result)
  ]}
 RigVMFunctionLibrary {nodes [
   Controllers = RigVMCollapseNode(Debug=False, ExecuteContext=())
   DeltaRotation = RigVMCollapseNode(ExecuteContext=())
   DeltaRotator = RigVMCollapseNode(ExecuteContext=(), debugRotation=False, onlyTwist=False, twistAxis=(X=0,Y=0,Z=0))
  ]}
 RigVMModel {nodes [
   RigUnit_BeginExecution = &RigUnit_BeginExecution()
   RigVMFunction_Sequence = &RigVMFunction_Sequence(ExecuteContext=@RigUnit_BeginExecution.ExecutePin)
   Controllers = RigVMFunctionReferenceNode(Debug=false, ExecuteContext=@RigVMFunction_Sequence.A)
   DeltaRotator = RigVMFunctionReferenceNode(
    child=head
    debugRotation=False
    max=-40
    onlyTwist=False
    parent=head_null
    parentType=Null
    premult=1
    primAxis=2
    twistAxis=(X=1,Y=0,Z=0)
    ExecuteContext=@RigVMFunction_Sequence.B
   )
   RigVMFunction_AnimEvalRichCurve = &RigVMFunction_AnimEvalRichCurve(
    Curve="(EditorCurveData=(Keys=((InterpMode=RCIM_Linear,TangentMode=RCTM_Auto,TangentWeightMode=RCTWM_WeightedNone,Time=0,Value=0.4,ArriveTangent=0,ArriveTangentWeight=0,LeaveTangent=0,LeaveTangentWeight=0),(InterpMode=RCIM_Cubic,TangentMode=RCTM_Auto,TangentWeigh ...(truncated)"
    SourceMinimum=0
    TargetMinimum=0
    Value=@DeltaRotator.outputWeight
   )
   > &RigUnit_ModifyTransforms(
    ItemToModify="((Item=(Type=Bone,Name=\"base_lowerAntenna_00_l\"),Transform=(Rotation=(X=0,Y=0,Z=0.7071,W=0.7071),Translation=(X=0,Y=0,Z=0),Scale3D=(X=1,Y=1,Z=1))),(Item=(Type=Bone,Name=\"base_lowerAntenna_00_r\"),Transform=(Rotation=(X=0,Y=0,Z=-0.5,W=0.866),Translation=(X=0 ...(truncated)"
    WeightMinimum=0
    Weight=RigVMDispatchNode(False=@DeltaRotator.outputWeight, True=@RigVMFunction_AnimEvalRichCurve.Result, Condition=RigVMVariableNode() {variable=IsCrouching}.Value) {template="DISPATCH_RigVMDispatch_If(in Condition,in True,in False,out Result)"}.Result
   )
   AlphaInterp = &RigVMFunction_AlphaInterp(InterpSpeedDecreasing=2, InterpSpeedIncreasing=8, Scale=1, Value=0, bClampResult=False)
  ]}
]
hierarchy [
 root {loc=<0 0 0>, rot=<0 0 0 1>}
 pelvis {parent=root, loc=<0 -0.7071 93.1191>, rot=<0.0127 0.707 -0.0127 -0.707>}
 spine_01 {parent=pelvis, loc=<6.7494 -0.6388 0>, rot=<0 0 0.034 -0.9994>}
 spine_02 {parent=spine_01, loc=<7.2518 -0.035 0>, rot=<0 0 0.0238 -0.9997>}
 spine_03 {parent=spine_02, loc=<6.0104 0 0>, rot=<0 0 0.0149 -0.9999>}
 spine_04 {parent=spine_03, loc=<8.1024 0 0>, rot=<0 0 -0.0132 -0.9999>}
 spine_05 {parent=spine_04, loc=<10.0146 0 0>, rot=<0 0 -0.0015 -1>}
 neck_01 {parent=spine_05, loc=<11.3138 0.1529 0>, rot=<0 0 -0.2003 -0.9797>}
 neck_02 {parent=neck_01, loc=<4.5627 0 0>, rot=<0 0 0.0504 -0.9987>}
 head {parent=neck_02, loc=<5.3914 0 0>, rot=<0 0 0.115 -0.9934>}
 base_upperAntenna_00_l {parent=head, loc=<11.1421 -9.0632 -3.9677>, rot=<0.2074 -0.1679 0.8766 0.4006>}
 dyn_upperAntenna_00_l {parent=base_upperAntenna_00_l, loc=<0 0 0>, rot=<0 0 0 1>}
 dyn_upperAntenna_01_l {parent=dyn_upperAntenna_00_l, loc=<-14.7282 -0.0001 0>, rot=<0 0 0 1>}
 base_lowerAntenna_00_l {parent=head, loc=<3.0781 -9.1013 -3.9492>, rot=<-0.2854 0.0469 -0.5531 -0.7813>}
 dyn_lowerAntenna_00_l {parent=base_lowerAntenna_00_l, loc=<0 0 0>, rot=<0 0 0 1>}
 dyn_lowerAntenna_01_l {parent=dyn_lowerAntenna_00_l, loc=<-9.0438 -0.0003 0.0001>, rot=<0 0 0 1>}
 base_lowerAntenna_00_r {parent=head, loc=<3.0779 -9.1013 3.9492>, rot=<0.5531 -0.7813 0.2854 0.0469>}
 dyn_lowerAntenna_00_r {parent=base_lowerAntenna_00_r, loc=<0 0 0>, rot=<0 0 0 1>}
 dyn_lowerAntenna_01_r {parent=dyn_lowerAntenna_00_r, loc=<9.0437 0 0>, rot=<0 0 0 1>}
 base_upperAntenna_00_r {parent=head, loc=<11.1416 -9.0632 3.9677>, rot=<0.8766 -0.4006 0.2074 0.1679>}
 dyn_upperAntenna_00_r {parent=base_upperAntenna_00_r, loc=<0 0 0>, rot=<0 0 0 1>}
 dyn_upperAntenna_01_r {parent=dyn_upperAntenna_00_r, loc=<14.7283 0 0>, rot=<0 0 0 1>}
 head_null {parent=neck_02, type=Null, loc=<5.3914 0 0>, rot=<0 0 0.115 -0.9934>}
 neck_01_null {type=Null, loc=<0 -4.3237 142.4487>, rot=<0.1137 0.6979 -0.1137 -0.6979>}
 neck_01_ctrl {controlType=EULER_TRANSFORM, parent=neck_01_null, type=Control, loc=<4.1271 -4.3018 142.4487>, rot=<-0.1488 -0.6913 0.0784 0.7027>}
 neck_02_null {parent=neck_01_ctrl, type=Null, loc=<4.5627 0 0>, rot=<0 0 -0.0504 0.9987>}
 neck_02_ctrl {controlType=EULER_TRANSFORM, parent=neck_02_null, type=Control, loc=<-149.8581 -22.259 143.3615>, rot=<-0.0598 -0.9852 0.1587 -0.0259>}
 head_null_2 {parent=neck_02_ctrl, type=Null, loc=<5.3914 0 0>, rot=<0 0 0.115 -0.9934>}
 head_ctrl {controlType=EULER_TRANSFORM, parent=head_null_2, type=Control, loc=<15.3569 -73.5387 12.2303>, rot=<0.0041 -0.0241 0.0531 0.9983>}
]
variables [IsCrouching {type=bool, default=False}]
