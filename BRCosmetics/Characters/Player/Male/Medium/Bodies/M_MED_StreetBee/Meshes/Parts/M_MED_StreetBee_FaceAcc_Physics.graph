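# Physics asset for the face accessory mesh of the StreetBee male medium player body, giving its four antennae secondary motion.
# Capsule bodies cover the head, a small base bone per antenna and two dynamic segments along each antenna.
# Each base is angularly locked to the head, and the segments chain outward from it.
# Segment joints allow only a few degrees of swing, tighter toward the tip, with zero-degree twist limits.
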PhysicsAsset /BRCosmetics/Characters/Player/Male/Medium/Bodies/M_MED_StreetBee/Meshes/Parts/M_MED_StreetBee_FaceAcc_Physics
bodies {
 base_lowerAntenna_00_l [{shapeType=Capsule, length=0.505, radius=0.505, shapeName=base_lowerAntenna_00_l_capsule, rotation=<rot 0 -90 0>}]
 base_lowerAntenna_00_r [{shapeType=Capsule, length=0.505, radius=0.505, shapeName=base_lowerAntenna_00_r_capsule, rotation=<rot 0 -90 0>}]
 base_upperAntenna_00_l [{shapeType=Capsule, length=0.505, radius=0.505, shapeName=base_upperAntenna_00_l_capsule, rotation=<rot 0 -90 0>}]
 base_upperAntenna_00_r [{shapeType=Capsule, length=0.505, radius=0.505, shapeName=base_upperAntenna_00_r_capsule, rotation=<rot 0 -90 0>}]
 dyn_lowerAntenna_00_l [{shapeType=Capsule, length=3.9761, radius=1.3216, shapeName=dyn_lowerAntenna_00_l_capsule, center=<-2.617 0 0>, rotation=<rot 90 0 0>}]
 dyn_lowerAntenna_00_r [{shapeType=Capsule, length=3.9771, radius=1.3216, shapeName=dyn_lowerAntenna_00_r_capsule, center=<2.6181 0 0>, rotation=<rot 90 0 0>}]
 dyn_lowerAntenna_01_l [{shapeType=Capsule, length=3.2086, radius=0.9617, shapeName=dyn_lowerAntenna_01_l_capsule, center=<-2.5663 0 0>, rotation=<rot 90 0 0>}]
 dyn_lowerAntenna_01_r [{shapeType=Capsule, length=3.2087, radius=0.9617, shapeName=dyn_lowerAntenna_01_r_capsule, center=<2.5663 0 0>, rotation=<rot 90 0 0>}]
 dyn_upperAntenna_00_l [{shapeType=Capsule, length=7.4002, radius=1.3261, shapeName=dyn_upperAntenna_00_l_capsule, center=<-6.0465 0 0>, rotation=<rot 90 0 0>}]
 dyn_upperAntenna_00_r [{shapeType=Capsule, length=7.4002, radius=1.3261, shapeName=dyn_upperAntenna_00_r_capsule, center=<6.0468 0 0>, rotation=<rot 90 0 0>}]
 dyn_upperAntenna_01_l [{shapeType=Capsule, length=3.6324, radius=0.9522, shapeName=dyn_upperAntenna_01_l_capsule, center=<-2.7154 0 0>, rotation=<rot 90 0 0>}]
 dyn_upperAntenna_01_r [{shapeType=Capsule, length=3.6324, radius=0.9522, shapeName=dyn_upperAntenna_01_r_capsule, center=<2.7157 0 0>, rotation=<rot 90 0 0>}]
 head [{shapeType=Capsule, length=1, radius=1}]
}
constraints [
 {
  bone1Name=base_lowerAntenna_00_l
  bone2Name=head
  swing1LimitDegrees=45
  swing1Motion=Locked
  swing2LimitDegrees=45
  swing2Motion=Locked
  twistLimitDegrees=45
  twistMotion=Locked
 }
 {bone1Name=dyn_lowerAntenna_00_l, bone2Name=base_lowerAntenna_00_l, swing1LimitDegrees=5, swing2LimitDegrees=5, twistLimitDegrees=0}
 {bone1Name=dyn_lowerAntenna_01_l, bone2Name=dyn_lowerAntenna_00_l, swing1LimitDegrees=3, swing2LimitDegrees=3, twistLimitDegrees=0}
 {
  bone1Name=base_lowerAntenna_00_r
  bone2Name=head
  swing1LimitDegrees=45
  swing1Motion=Locked
  swing2LimitDegrees=45
  swing2Motion=Locked
  twistLimitDegrees=45
  twistMotion=Locked
 }
 {bone1Name=dyn_lowerAntenna_00_r, bone2Name=base_lowerAntenna_00_r, swing1LimitDegrees=5, swing2LimitDegrees=5, twistLimitDegrees=0}
 {bone1Name=dyn_lowerAntenna_01_r, bone2Name=dyn_lowerAntenna_00_r, swing1LimitDegrees=3, swing2LimitDegrees=3, twistLimitDegrees=0}
 {
  bone1Name=base_upperAntenna_00_l
  bone2Name=head
  swing1LimitDegrees=45
  swing1Motion=Locked
  swing2LimitDegrees=45
  swing2Motion=Locked
  twistLimitDegrees=45
  twistMotion=Locked
 }
 {bone1Name=dyn_upperAntenna_00_l, bone2Name=base_upperAntenna_00_l, swing1LimitDegrees=5, swing2LimitDegrees=5, twistLimitDegrees=0}
 {bone1Name=dyn_upperAntenna_01_l, bone2Name=dyn_upperAntenna_00_l, swing1LimitDegrees=3, swing2LimitDegrees=3, twistLimitDegrees=0}
 {
  bone1Name=base_upperAntenna_00_r
  bone2Name=head
  swing1LimitDegrees=45
  swing1Motion=Locked
  swing2LimitDegrees=45
  swing2Motion=Locked
  twistLimitDegrees=45
  twistMotion=Locked
 }
 {bone1Name=dyn_upperAntenna_00_r, bone2Name=base_upperAntenna_00_r, swing1LimitDegrees=5, swing2LimitDegrees=5, twistLimitDegrees=0}
 {bone1Name=dyn_upperAntenna_01_r, bone2Name=dyn_upperAntenna_00_r, swing1LimitDegrees=3, swing2LimitDegrees=3, twistLimitDegrees=0}
]
